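# Skeletal-mesh Default Lit material for the sand on the Phantom backpack accessory, giving it a flowing, glittering look.
# A UV-gradient mask blends a static metallic-flake normal map with a panning copy, confining motion to a masked region.
# The blended normal, transformed and dotted with the camera vector, is raised to a power to add tinted, view-dependent highlights over the sand color.
# Attributes pass through MF_TechArtOverrides_accessories before output.
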Material /BRCosmetics/Accessories/FORT_Backpacks/M_MED_Phantom_Backpack/Materials/M_MED_Phantom_Backpack_Sand
graph {
 properties {
  bUseMaterialAttributes=true
  bUsedWithMeshDeformer=true
  bUsedWithSkeletalMesh=true
  pixelDepthOffsetMode=PDOM_Legacy
  usedShadingModels=Default Lit
 }
 nodes [
  n1 = VectorParameter() {parameterName=SandColor, defaultValue=<rgba 0.5 0.3731 0.176 1>}
  n2 = CameraVectorWS()
  n3 = TextureCoordinate()
  n4 = ScalarParameter() {defaultValue=3, parameterName=Sand_Scale}
  n5 = Multiply(A=@n3, B=ScalarParameter() {defaultValue=3.7129, parameterName=Sand_Scale_2})
  n6 = AppendVector(A=Constant(), B=ScalarParameter() {defaultValue=-0.2, parameterName=Sand_PanSpeed})
  n7 = TextureCoordinate()
  n8 = ComponentMask(None=@n7) {g=true}
  n9 = Saturate(None=LinearInterpolate(A=ScalarParameter() {parameterName=Mask1}, B=ScalarParameter() {defaultValue=35.7865, parameterName=Mask2}, Alpha=@n8))
  n10 = Saturate(None=LinearInterpolate(A=ScalarParameter() {defaultValue=3.9643, parameterName=Mask3}, B=ScalarParameter() {parameterName=Mask4}, Alpha=@n8))
  n11 = LinearInterpolate(
   A=ScalarParameter() {defaultValue=-166.9875, parameterName=Mask5}
   B=ScalarParameter() {defaultValue=80, parameterName=Mask6}
   Alpha=ComponentMask(None=@n7) {r=true}
  )
  n12 = LinearInterpolate(
   A=TextureSample(UVs=Multiply(A=@n3, B=@n4)) {samplerType=SAMPLERTYPE_Normal, texture=/Game/Weapons/WeaponSkins/Wraps/Textures/T_Wrap_metallicFlakes_Small}.RGB
   B=TextureSample(UVs=Panner(Coordinate=@n5, Speed=@n6)) {samplerType=SAMPLERTYPE_Normal, texture=&T_Wrap_metallicFlakes_Small}.RGB
   Alpha=Multiply(A=Multiply(A=@n9, B=@n10), B=Saturate(None=@n11))
  )
  n13 = Normalize(VectorInput=Transform(None=@n12))
  n14 = Multiply(
   A=VectorParameter() {parameterName=SandColor2, defaultValue=<rgba 0.375 0.3253 0.248 1>}.RGB
   B=Power(Base=DotProduct(A=@n2, B=@n13), Exp=ScalarParameter() {defaultValue=5.9566, parameterName=Power})
  )
  n15 = SetMaterialAttributes!(
   MaterialAttributes
   Base Color=NamedRerouteDeclaration(None=Add(A=@n1.RGB, B=@n14)) {name=baseColor, nodeColor=<rgba 0.2656 1 0 1>}
   Specular=NamedRerouteDeclaration(None=ScalarParameter() {defaultValue=1, parameterName=Specular}) {name=specular, nodeColor=<rgba 1 0 0.75 1>}
   Roughness=NamedRerouteDeclaration(None=ScalarParameter() {defaultValue=0.4857, parameterName=Roughness}) {name=roughness, nodeColor=<rgba 0 1 0.4375 1>}
   Normal=NamedRerouteDeclaration(None=@n12) {name=normal, nodeColor=<rgba 1 0 0.5156 1>}
  ) {attributeSetTypes [
    #1
    #2
    #3
    #4
   ]}
  MaterialFunctionCall(In=@n15) {materialFunction=&MF_TechArtOverrides_accessories}
 ]
}
