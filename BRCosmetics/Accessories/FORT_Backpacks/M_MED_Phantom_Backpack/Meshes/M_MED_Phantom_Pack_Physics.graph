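# Physics asset for the Phantom backpack cosmetic, defining bodies and joints for its bandage bones.
# From the pack, main base and main bodies are joined by a free then a locked angular joint, branching through free joints to upper and lower masters.
# Each master holds a capsule bandage body with tight angular limits.
# A four-box strip hangs off the lower bandage, with swing limits widening after its first joint.
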PhysicsAsset /BRCosmetics/Accessories/FORT_Backpacks/M_MED_Phantom_Backpack/Meshes/M_MED_Phantom_Pack_Physics
bodies {
 dyn_bandage_01 [{shapeType=Box, extentX=7.5, extentY=4, extentZ=1.8, center=<3 0 0>}]
 dyn_bandage_02 [{shapeType=Box, extentX=7.5, extentY=4, extentZ=1.8, center=<3 0 0>}]
 dyn_bandage_03 [{shapeType=Box, extentX=7.5, extentY=4, extentZ=1.8, center=<3 0 0>}]
 dyn_bandage_base [{shapeType=Box, extentX=7.5, extentY=4, extentZ=1.8, center=<4 0 0>}]
 dyn_bandage_main [{shapeType=Capsule, length=1, radius=1}]
 dyn_bandage_main_base [{shapeType=Capsule, length=1, radius=1}]
 dyn_lw_bandage [{shapeType=Capsule, length=4.25, radius=4.75, center=<-3.3031 0 -0.0001>, rotation=<rot 90 0 0>}]
 dyn_lw_bandage_MSTR [{shapeType=Capsule, length=1, radius=1}]
 dyn_up_bandage [{shapeType=Capsule, length=4, radius=5, center=<3.5585 0 0>, rotation=<rot 90 0 0>}]
 dyn_up_bandage_MSTR [{shapeType=Capsule, length=1, radius=1}]
 pack [{shapeType=Capsule, length=1, radius=1}]
}
constraints [
 {bone1Name=dyn_bandage_01, bone2Name=dyn_bandage_base, swing1LimitDegrees=30, swing2LimitDegrees=30, twistLimitDegrees=15}
 {bone1Name=dyn_bandage_02, bone2Name=dyn_bandage_01, swing1LimitDegrees=30, swing2LimitDegrees=30, twistLimitDegrees=15}
 {bone1Name=dyn_bandage_03, bone2Name=dyn_bandage_02, swing1LimitDegrees=30, swing2LimitDegrees=30, twistLimitDegrees=15}
 {bone1Name=dyn_bandage_base, bone2Name=dyn_lw_bandage, swing1LimitDegrees=15, swing2LimitDegrees=15, twistLimitDegrees=15}
 {bone1Name=dyn_up_bandage, bone2Name=dyn_up_bandage_MSTR, swing1LimitDegrees=10, swing2LimitDegrees=10, twistLimitDegrees=5}
 {bone1Name=dyn_lw_bandage, bone2Name=dyn_lw_bandage_MSTR, swing1LimitDegrees=10, swing2LimitDegrees=10, twistLimitDegrees=5}
 {
  bone1Name=dyn_up_bandage_MSTR
  bone2Name=dyn_bandage_main
  swing1LimitDegrees=45
  swing1Motion=Free
  swing2LimitDegrees=45
  swing2Motion=Free
  twistLimitDegrees=45
  twistMotion=Free
 }
 {
  bone1Name=dyn_lw_bandage_MSTR
  bone2Name=dyn_bandage_main
  swing1LimitDegrees=45
  swing1Motion=Free
  swing2LimitDegrees=45
  swing2Motion=Free
  twistLimitDegrees=45
  twistMotion=Free
 }
 {
  bone1Name=dyn_bandage_main_base
  bone2Name=pack
  swing1LimitDegrees=45
  swing1Motion=Free
  swing2LimitDegrees=45
  swing2Motion=Free
  twistLimitDegrees=45
  twistMotion=Free
 }
 {
  bone1Name=dyn_bandage_main
  bone2Name=dyn_bandage_main_base
  swing1LimitDegrees=45
  swing1Motion=Locked
  swing2LimitDegrees=45
  swing2Motion=Locked
  twistLimitDegrees=45
  twistMotion=Locked
 }
]
