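# Animation Blueprint for the Phantom backpack cosmetic, animating its dangling bandages with procedural flutter and physics.
# The update event generates Perlin noise offsets and rotations, and maps downward pawn velocity to a lift alpha.
# The output chain applies bandage translations and lift before RigidBody simulation in root bone space, then layers noisy rotations on top.
# A separate AnimDynamics and RigidBody chain using ShardFloat offsets is disconnected from the output pose.
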AnimBlueprint /BRCosmetics/Accessories/FORT_Backpacks/M_MED_Phantom_Backpack/Meshes/M_MED_Phantom_Pack_AnimBP
target_skeleton=&/M_MED_Phantom_Pack_Skeleton
graphs [
 AnimGraph {nodes [
   n1 = "Transform (Modify) Bone - Bone: dyn_up_bandage_MSTR"(Alpha=1, Translation=Get ShardFloat_A() -> (ShardFloat_A:Vector), Rotation="0,0,0", Scale="1,1,1") {
    translationMode=BMM_Additive
    translationSpace=BCS_BoneSpace
    alphaBoolBlend {blendOption=Linear}
    boneToModify {boneName=dyn_up_bandage_MSTR}
   }
   n2 = "Transform (Modify) Bone - Bone: dyn_lw_bandage_MSTR"(ComponentPose=@n1, Alpha=1, Translation=Get ShardFloat_B() -> (ShardFloat_B:Vector), Rotation="0,0,0", Scale="1,1,1") {
    translationMode=BMM_Additive
    translationSpace=BCS_BoneSpace
    alphaBoolBlend {blendOption=Linear}
    boneToModify {boneName=dyn_lw_bandage_MSTR}
   }
   n3 = Anim Dynamics - Bone: dyn_up_bandage(ComponentPose=@n2, Alpha=1, ExternalForce="0,0,0") {
    angularSpringConstant=100
    bAngularSpring=true
    bGravityOverrideInSimSpace=true
    bLinearSpring=true
    bOverrideLinearDamping=true
    bUseGravityOverride=true
    linearDampingOverride=0.99
    linearSpringConstant=180
    alphaBoolBlend {blendOption=Linear}
    boundBone {boneName=dyn_up_bandage}
    componentAppliedLinearAccClamp=<3000 3000 3000>
    componentLinearAccScale=<0.25 0.25 0.1>
    componentLinearVelScale=<0.2 0.2 0.2>
    physicsBodyDefinitions [{
      sphereCollisionRadius=0
      boundBone {boneName=dyn_up_bandage}
      boxExtents=<8 8 8>
      constraintSetup {angularLimitsMax=<5 5 5>, angularLimitsMin=<-5 -5 -5>, angularTarget=<0 0 0>, linearAxesMax=<2 0.8 0.8>, linearAxesMin=<-2 -0.8 -0.8>}
      localJointOffset=<0 0 0>
     }]
   }
   n4 = Anim Dynamics - Bone: dyn_lw_bandage(ComponentPose=@n3, Alpha=1, ExternalForce="0,0,0") {
    angularSpringConstant=100
    bAngularSpring=true
    bGravityOverrideInSimSpace=true
    bLinearSpring=true
    bOverrideLinearDamping=true
    bUseGravityOverride=true
    linearDampingOverride=0.995
    linearSpringConstant=180
    alphaBoolBlend {blendOption=Linear}
    boundBone {boneName=dyn_lw_bandage}
    componentAppliedLinearAccClamp=<3000 3000 3000>
    componentLinearAccScale=<0.35 0.35 0.1>
    componentLinearVelScale=<0.8 0.8 0.8>
    physicsBodyDefinitions [{
      sphereCollisionRadius=0
      boundBone {boneName=dyn_lw_bandage}
      boxExtents=<8 8 8>
      constraintSetup {angularLimitsMax=<5 5 5>, angularLimitsMin=<-5 -5 -5>, angularTarget=<0 0 0>, linearAxesMax=<2 0.8 0.8>, linearAxesMin=<-2 -0.8 -0.8>}
      localJointOffset=<0.5 0 0>
     }]
   }
   RigidBody(ComponentPose=@n4) {
    bClampLinearTranslationLimitToRefPose=true
    bOverrideWorldGravity=true
    overridePhysicsAsset=&/M_MED_Phantom_Pack_Physics
    alphaBoolBlend {blendOption=Linear}
    componentAppliedLinearAccClamp=<3000 3000 3000>
    componentLinearAccScale=<0.6 0.6 0.6>
    componentLinearVelScale=<0.1 0.1 0.1>
    overrideWorldGravity {z=-300}
   }
   AnimGraphNode_SubInput_0 = Input Pose()
   n5 = "Transform (Modify) Bone - Bone: dyn_up_bandage_MSTR"(
    ComponentPose=Local To Component(LocalPose=@AnimGraphNode_SubInput_0)
    Alpha=1
    Translation=Get TranNoiseUp() -> (TranNoiseUp:Vector)
    Rotation="0,0,0"
    Scale="1,1,1"
   ) {
    rotationMode=BMM_Additive
    rotationSpace=BCS_BoneSpace
    translationMode=BMM_Additive
    translationSpace=BCS_BoneSpace
    alphaBoolBlend {blendOption=Linear}
    boneToModify {boneName=dyn_up_bandage_MSTR}
   }
   n6 = "Transform (Modify) Bone - Bone: dyn_lw_bandage_MSTR"(ComponentPose=@n5, Alpha=1, Translation=Get TranNoiseDn() -> (TranNoiseDn:Vector), Rotation="0,0,0", Scale="1,1,1") {
    translationMode=BMM_Additive
    translationSpace=BCS_BoneSpace
    alphaBoolBlend {blendOption=Linear}
    boneToModify {boneName=dyn_lw_bandage_MSTR}
   }
   n7 = "Transform (Modify) Bone - Bone: dyn_bandage_main_base"(ComponentPose=@n6, Alpha=Get jumpAlpha() -> (jumpAlpha:double), Translation="3,0,0", Rotation="0,0,0", Scale="1,1,1") {
    rotationMode=BMM_Additive
    rotationSpace=BCS_BoneSpace
    translationMode=BMM_Additive
    translationSpace=BCS_BoneSpace
    alphaBoolBlend {blendOption=Linear}
    alphaScaleBiasClamp {bInterpResult=true}
    boneToModify {boneName=dyn_bandage_main_base}
    translation {x=3}
   }
   n8 = RigidBody(ComponentPose=@n7) {
    bClampLinearTranslationLimitToRefPose=true
    bOverrideWorldGravity=true
    overridePhysicsAsset=&M_MED_Phantom_Pack_Physics
    simulationSpace=BaseBoneSpace
    alphaBoolBlend {blendOption=Linear}
    baseBoneRef {boneName=root}
    componentAppliedLinearAccClamp=<3000 3000 3000>
    componentLinearAccScale=<0.2 0.2 0.15>
    componentLinearVelScale=<0.2 0.2 0.15>
    overrideWorldGravity {z=-100}
   }
   n9 = "Transform (Modify) Bone - Bone: dyn_bandage_01"(ComponentPose=@n8, Alpha=1, Translation="0,0,1", Rotation="50,0,0", Scale="1,1,1") {
    rotationMode=BMM_Additive
    rotationSpace=BCS_BoneSpace
    translationMode=BMM_Additive
    translationSpace=BCS_BoneSpace
    alphaBoolBlend {blendOption=Linear}
    boneToModify {boneName=dyn_bandage_01}
    rotation {pitch=50}
    translation {z=1}
   }
   n10 = Get RotateNoiseA() -> (RotateNoiseA:Rotator)
   n11 = "Transform (Modify) Bone - Bone: dyn_bandage_base"(ComponentPose=@n9, Alpha=0.4, Translation="0,0,0", Rotation=@n10, Scale="1,1,1") {
    rotationMode=BMM_Additive
    rotationSpace=BCS_BoneSpace
    translationMode=BMM_Additive
    translationSpace=BCS_BoneSpace
    alphaBoolBlend {blendOption=Linear}
    boneToModify {boneName=dyn_bandage_base}
   }
   n12 = "Transform (Modify) Bone - Bone: dyn_bandage_01"(ComponentPose=@n11, Alpha=1, Translation="0,0,0", Rotation=@n10, Scale="1,1,1") {
    rotationMode=BMM_Additive
    rotationSpace=BCS_BoneSpace
    translationMode=BMM_Additive
    translationSpace=BCS_BoneSpace
    alphaBoolBlend {blendOption=Linear}
    boneToModify {boneName=dyn_bandage_01}
   }
   n13 = "Transform (Modify) Bone - Bone: dyn_bandage_02"(ComponentPose=@n12, Alpha=1, Translation="0,0,0", Rotation=Get RotateNoiseB() -> (RotateNoiseB:Rotator), Scale="1,1,1") {
    rotationMode=BMM_Additive
    rotationSpace=BCS_BoneSpace
    translationMode=BMM_Additive
    translationSpace=BCS_BoneSpace
    alphaBoolBlend {blendOption=Linear}
    boneToModify {boneName=dyn_bandage_02}
   }
   n14 = "Transform (Modify) Bone - Bone: dyn_bandage_03"(ComponentPose=@n13, Alpha=1, Translation="0,0,0", Rotation=Get RotateNoiseC() -> (RotateNoiseC:Rotator), Scale="1,1,1") {
    rotationMode=BMM_Additive
    rotationSpace=BCS_BoneSpace
    translationMode=BMM_Additive
    translationSpace=BCS_BoneSpace
    alphaBoolBlend {blendOption=Linear}
    boneToModify {boneName=dyn_bandage_03}
   }
   Output Pose(Result=Component To Local(ComponentPose=@n14)) {layerGroup=None}
  ]}
 EventGraph {nodes [
   Event BlueprintUpdateAnimation()
   > n1 = Sequence!(execute) -> (then_0, then_1, then_2:Exec, then_3:Exec, then_4:Exec)
   > n2 = Sequence!(execute) -> (then_0, then_1, then_2)
   > n3 = Noise_FNC!(execute, self, frequency:double=0.35, amp:double=6) -> (then, output:double)
   n4 = Noise_FNC(execute=@n2.then_1, frequency=0.3, amp=-4.5)
   n5 = Noise_FNC(execute=@n2.then_2, frequency=0.3, amp=3)
   Set TranNoiseDn(execute=@n4.then, TranNoiseDn:Vector=MakeVector(X=@n4, Y=0, Z=@n5)) -> (Output_Get:Vector)
   n6 = Sequence(execute=@n1.then_1)
   > n7 = Noise_FNC(frequency=0.3, amp=25)
   n8 = Noise_FNC(execute=@n6.then_1, frequency=0.4, amp=30)
   n9 = Noise_FNC(execute=@n6.then_2, frequency=0.35, amp=15)
   > Set RotateNoiseA(RotateNoiseA:Rotator=MakeRotator(Roll=@n7, Pitch=@n8, Yaw=@n9)) -> (Output_Get:Rotator)
   n10 = Sequence(execute=@n1.then_2)
   > n11 = Noise_FNC(frequency=0.35, amp=20)
   n12 = Noise_FNC(execute=@n10.then_1, frequency=0.3, amp=30)
   n13 = Noise_FNC(execute=@n10.then_2, frequency=0.35, amp=15)
   > Set RotateNoiseB(RotateNoiseB:Rotator=MakeRotator(Roll=@n11, Pitch=@n12, Yaw=@n13)) -> (Output_Get:Rotator)
   n14 = Sequence(execute=@n1.then_3)
   > n15 = Noise_FNC(frequency=0.2, amp=50)
   n16 = Noise_FNC(execute=@n14.then_1, frequency=0.2, amp=50)
   n17 = Noise_FNC(execute=@n14.then_2, frequency=0.2, amp=30)
   > Set RotateNoiseC(RotateNoiseC:Rotator=MakeRotator(Roll=@n15, Pitch=@n16, Yaw=@n17)) -> (Output_Get:Rotator)
   Set jumpAlpha(execute=@n1.then_4, jumpAlpha:double=MapRangeClamped(Value=Get PawnVelocityZ(), InRangeA=-700, InRangeB=0, OutRangeA=1, OutRangeB=0)) -> (Output_Get:double)
   Reroute Node()
   n18 = GetGameTimeInSeconds()
   n19 = float * float(A=float * float(A=1, B=0.85), B=@n18)
   n20 = float * float(A=float * float(A=1, B=0.8), B=@n18)
   n21 = float * float(A=float * float(A=1, B=0.55), B=@n18)
   n22 = MakeVector(
    X=float * float(A="Sin (Radians)"(A=@n21), B=2)
    Y=float * float(A="Sin (Radians)"(A=@n19), B=1.2)
    Z=float * float(A="Sin (Radians)"(A=@n20), B=1.5)
   )
   > Set ShardFloat_A(ShardFloat_A:Vector=@n22) -> (Output_Get:Vector)
   > Set ShardFloat_B(ShardFloat_B:Vector=vector * float(A=@n22, B=-0.8)) -> (Output_Get:Vector)
   TryGetPawnOwner()
   MapRangeClamped(Value=@n8, InRangeA=-20, InRangeB=20, OutRangeA=-10, OutRangeB=40)
   Delay(Duration=0.2)
   Set TranNoiseUp(execute=@n5.then, TranNoiseUp:Vector=MakeVector(X=@n3, Y=0, Z=@n5)) -> (Output_Get:Vector)
  ]}
 Noise_FNC {nodes [
   n1 = Noise_FNC!() -> (then, frequency:double=~, amp:double=~) {is=title K2Node_FunctionEntry}
   n2 = float * float(A=@n1, B=GetGameTimeInSeconds())
   n3 = float * float(A="Sin (Radians)"(A=PerlinNoise1D!(Value:float=@n2) -> (ReturnValue:float)), B=@n1.amp)
   > Return Node!(execute, output:double=@n3)
   PerlinNoise1D(Value=@n3)
  ]}
]
variables [
 jumpAlpha {type=number}
 RotateNoiseA {type=Rotator}
 RotateNoiseB {type=Rotator}
 RotateNoiseC {type=Rotator}
 ShardFloat_A {type=Vector}
 ShardFloat_B {type=Vector}
 TranNoiseDn {type=Vector}
 TranNoiseUp {type=Vector}
]
